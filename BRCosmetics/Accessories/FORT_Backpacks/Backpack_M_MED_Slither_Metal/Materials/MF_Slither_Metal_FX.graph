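# Material function for the Slither Metal backpack that adds animated flame and shine emissive to incoming material attributes.
# The flame mask uses bump-offset parallax and scrolling cloud-noise distortion, with flicker from three summed time sines.
# A rotated, panning gradient distorted by water noise supplies the shine.
# Named reroutes carry normal, specular/roughness-derived variation and Skin-based logo masks, while a sphere mask confines the glow.
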MaterialFunction /BRCosmetics/Accessories/FORT_Backpacks/Backpack_M_MED_Slither_Metal/Materials/MF_Slither_Metal_FX
graph {nodes [
  n1 = Time()
  n2 = VectorParameter() {group=SlitherMetalFX, parameterName=FlameFadeTimes, defaultValue=<rgba 0.23 1.2 3.33 0>}
  n3 = Sine(None=Multiply(A=@n1, B=@n2.B) {constB=3.33})
  n4 = Constant() {r=0.5}
  Add(A=Multiply(A=@n3, B=@n4), B=@n4)
  n5 = TextureCoordinate()
  n6 = MaterialFunctionCall(Float2=@n5) {materialFunction=&BreakOutFloat2Components}
  AppendVector(A=ComponentMask(None=@n5) {r=true}, B=Multiply(A=Add(A=@n6.R, B=@n6.R)) {constB=0.5})
  FunctionInput() {bUsePreviewValueAsDefault=true, inputName=GMap, inputType=FunctionInput_Scalar}
  n7 = GetMaterialAttributes(None=FunctionInput() {inputName=Material, inputType=FunctionInput_MaterialAttributes, sortPriority=2}) {attributeGetTypes [
    #1
    #2
    #3
    #4
   ]}
  n8 = NamedRerouteUsage()
  FunctionOutput(None=SetMaterialAttributes!(MaterialAttributes=@n7.MaterialAttributes, EmissiveColor=Add(A=@n7.EmissiveColor, B=@n8)) {attributeSetTypes [#1]})
  n9 = VectorParameter() {group=SlitherMetalFX, parameterName=FlameFadeRange, defaultValue=<rgba 0.02 1 0 0>}
  n10 = Sine(None=Multiply(A=@n1, B=@n2.R) {constB=0.23})
  n11 = Constant() {r=0.5}
  n12 = Multiply(A=Add(A=Multiply(A=@n10, B=@n11), B=@n11)) {constB=0.6}
  n13 = Sine(None=Multiply(A=@n1, B=@n2.G) {constB=1.2})
  n14 = Constant() {r=0.5}
  n15 = Multiply(A=Add(A=Multiply(A=@n13, B=@n14), B=@n14)) {constB=0.3}
  n16 = Saturate(None=Add(A=Add(A=@n12, B=@n15), B=Multiply(A=@n3) {constB=0.125}))
  LinearInterpolate(A=@n9.R, B=@n9.R, Alpha=@n16)
  Multiply(A=@n5, B=Constant2Vector() {g=-1, r=1})
  n17 = FunctionInput() {bUsePreviewValueAsDefault=true, inputName=Skin, inputType=FunctionInput_Scalar, sortPriority=1}
  NamedRerouteDeclaration(None=Saturate(None=Power(Base=@n17) {constExponent=2.2})) {name=LogoMask, nodeColor=<rgba 1 1 1 1>}
  NamedRerouteDeclaration(None=@n7.Normal) {name=Normal, nodeColor=<rgba 0 0.3446 1 1>}
  n18 = Multiply(A=@n7.Specular, B=OneMinus(None=@n7.Roughness))
  NamedRerouteDeclaration(None=Multiply(A=@n18, B=@n18)) {name=VariationMask, nodeColor=<rgba 0 1 0.6574 1>}
  n19 = Constant() {r=2}
  n20 = TextureSample(UVs=Panner(Coordinate=MaterialFunctionCall(Texture Scale=@n19) {materialFunction=&ScaleUVsByCenter}.UVs) {speedY=-0.5}) {texture=/Game/Global/Textures/Default/Noise/T_Noise_Cloud}
  n21 = ComponentMask(None=VectorParameter() {parameterName=FlameDistortion, defaultValue=<rgba 0.6 0.4 0 1>}.RGB) {g=true, r=true}
  n22 = Multiply(A=AppendVector(A=@n20.G, B=@n20.G), B=@n21) {constB=0.7}
  n23 = Multiply(A=@n21) {constB=0.5}
  n24 = Multiply(
   A=BumpOffset(Coordinate=TextureCoordinate(), Height=ScalarParameter() {defaultValue=-20, group=SlitherMetalFX, parameterName=BumpOffset, sliderMax=-50}) {referencePlane=0}
   B=Constant2Vector() {g=-1, r=1}
  )
  n25 = VectorParameter() {group=SlitherMetalFX, parameterName=FlameOffsetSize, defaultValue=<rgba 0.2 0.2 1.5 1.5>}
  n26 = ComponentMask(None=@n25.RGB) {g=true, r=true}
  n27 = NamedRerouteUsage()
  n28 = Add(
   A=MaterialFunctionCall(UVs=Add(A=@n24, B=@n26), Texture Scale=AppendVector(A=@n25.B, B=@n25.B)) {materialFunction=&ScaleUVsByCenter}.UVs
   B=Multiply(
    A=ComponentMask(None=@n27) {g=true, r=true}
    B=ScalarParameter() {defaultValue=0.05, group=SlitherMetalFX, parameterName=FlameNormalDistortion}
   )
  )
  n29 = LinearInterpolate(Alpha=NamedRerouteUsage()) {constA=4, constB=0}
  n30 = TextureSampleParameter2D!(UVs=Add(A=Subtract(A=@n22, B=@n23), B=Saturate(None=@n28)), Bias=@n29, Apply View MipBias) {
   group=SlitherMetalFX
   mipValueMode=TMVM_MipBias
   parameterName=FlameMask
   samplerSource=SSM_Clamp_WorldGroupSettings
   texture=/Game/WaterAndWind/Textures/WeaponFireVectorMask
  }
  n31 = VectorParameter() {group=SlitherMetalFX, parameterName=FlameColorIntensity, defaultValue=<rgba 1 0.2862 0.0242 2>}
  n32 = Multiply(A=Multiply(A=@n30.R, B=Multiply(A=@n31.RGB, B=@n31.RGB) {constB=3}), B=@n16)
  n33 = NamedRerouteUsage()
  n34 = Add(
   A=Rotator(Coordinate=@n5, Time=ScalarParameter() {group=SlitherMetalFX, parameterName=ShineRotation}) {speed=1}
   B=Multiply(A=ComponentMask(None=@n33) {g=true, r=true}, B=Constant() {r=0.05})
  )
  n35 = Panner(Coordinate=@n34, Speed=VectorParameter() {group=SlitherMetalFX, parameterName=ShineSpeed, defaultValue=<rgba 0 0.2 0 1>}.RGB)
  n36 = TextureSampleParameter2D!(UVs=Panner(Coordinate=@n34) {speedY=-0.025}, Apply View MipBias) {
   group=SlitherMetalFX
   parameterName=ShineDistorionNoise
   samplerType=SAMPLERTYPE_LinearColor
   texture=/Game/Weapons/WeaponSkins/Wraps/Textures/Beach/T_Noise_Water
  }
  n37 = Multiply(A=@n36.G, B=ScalarParameter() {defaultValue=0.7, group=SlitherMetalFX, parameterName=ShineDistortion, sliderMax=1}) {constB=0.7}
  n38 = VectorParameter() {group=SlitherMetalFX, parameterName=ShineColorIntensity, defaultValue=<rgba 0.5781 0.8026 1 2>}
  n39 = Multiply(
   A=TextureSampleParameter2D!(UVs=Add(A=@n35, B=@n37), Bias=@n29, Apply View MipBias) {mipValueMode=TMVM_MipBias, parameterName=ShineGradient, texture=/Game/Effects/Fort_Effects/Textures/Grad/gradient_healed}.R
   B=Multiply(A=@n38.RGB, B=@n38.RGB)
  ) {constB=5}
  n40 = Saturate(None=Add(A=ScalarParameter() {defaultValue=0.05, group=SlitherMetalFX, parameterName=MaskLowerValue, sliderMax=1}, B=NamedRerouteUsage()))
  n41 = SphereMask(A=TextureCoordinate(), B=Constant() {r=0.5}) {attenuationRadius=0.37, hardnessPercent=0}
  NamedRerouteDeclaration(None=Multiply(A=Add(A=@n32, B=@n39), B=Multiply(A=@n40, B=@n41))) {name=Emissive, nodeColor=<rgba 1 0.1648 0 1>}
 ]}
