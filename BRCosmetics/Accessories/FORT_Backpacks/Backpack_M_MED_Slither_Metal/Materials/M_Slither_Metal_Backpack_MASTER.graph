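# Master material for the Slither Metal backpack cosmetic, a Default Lit surface for skeletal meshes with mesh deformer support.
# It uses material attributes and delegates its shading to three chained material functions.
# The shared character PreFX function feeds the Slither Metal FX function, whose result enters the shared PostFX function alongside further PreFX outputs.
# The backpack-specific effect logic lives in MF_Slither_Metal_FX, outside this graph.
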Material /BRCosmetics/Accessories/FORT_Backpacks/Backpack_M_MED_Slither_Metal/Materials/M_Slither_Metal_Backpack_MASTER
graph {
 properties {
  bUseMaterialAttributes=true
  bUsedWithMeshDeformer=true
  bUsedWithSkeletalMesh=true
  pixelDepthOffsetMode=PDOM_Legacy
  usedShadingModels=Default Lit
 }
 nodes [
  n1 = MaterialFunctionCall() {materialFunction=&BaseFN_CharacterFunction_PreFX}
  MaterialFunctionCall(
   AO=@n1.AO
   Distance Blend=@n1.AO
   Fresnel=@n1.AO
   Gmap=@n1.AO
   Pre Skinned Local Position=@n1.AO
   Material=MaterialFunctionCall!(GMap=@n1.Gmap, Skin=@n1.Gmap, Material=@n1.Gmap) {materialFunction=&/MF_Slither_Metal_FX}.Result
  ) {materialFunction=&BaseFN_CharacterFunction_PostFX}
 ]
}
